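# Translucent unlit material for Niagara ribbons, named as the deploy trail of the male Stamina glider.
# Particle color RGB runs through MF_FX_EmissiveAdjustments for the glow.
# A separate mask multiplies particle alpha, the CylGradient texture and a procedural 4x(1-x) parabola on the first UV component, which fades to zero at both ends and peaks mid-way.
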Material /BRCosmetics/Gadgets/Assets/VinderTech_GliderChute/Glider_Stamina_Male/FX/Material/M_StaminaMale_DeployTrail
graph {
 properties {bUsedWithNiagaraRibbons=true, blendMode=BLEND_Translucent, pixelDepthOffsetMode=PDOM_Legacy, shadingModel=MSM_Unlit, usedShadingModels=Unlit}
 nodes [
  n1 = ParticleColor()
  MaterialFunctionCall(Emissive=@n1.RGB) {materialFunction=&MF_FX_EmissiveAdjustments}
  n2 = MaterialFunctionCall(Float2=TextureCoordinate()) {materialFunction=&BreakOutFloat2Components}
  n3 = TextureSample() {texture=/Game/Global/Textures/Default/Masks_and_Gradients/CylGradient}
  Multiply(A=Multiply(A=Multiply(A=@n2.R) {constB=4}, B=Multiply(A=@n1.A, B=@n3.R)), B=OneMinus(None=@n2.R))
 ]
}
