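# Niagara deploy effect for the Stamina Male glider, with separate gameplay and front-end visuals.
# In gameplay, SmallCloud fires a delayed burst of tumbling meshes gated by User.Deployed, with drag and gravity, writing a skeletal-mesh-offset position.
# In the front end, FEFakeTail instead spawns one mesh particle animated through curve-driven material parameters and a timer.
# A skeletal-mesh ribbon trail with time-of-day tinting is present but disabled.
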NiagaraSystem /BRCosmetics/Gadgets/Assets/VinderTech_GliderChute/Glider_Stamina_Male/FX/Niagara/NS_StaminaMale_Deploy
dependencies {
 usedDataInterfaces [&NiagaraDataInterfaceCurve, &NiagaraDataInterfaceColorCurve, &NiagaraDataInterfaceCamera]
 usedDynamicInputs [
  &Power
  &ClampFloat
  &DotProduct
  &ReturnCameraVector
 ]
 usedModules [
  &SystemState
  &InterpolateOverTime
  &SampleSkeletalMesh
  &self:SystemSpawnScript.SystemScriptSource.SystemScriptGraph.NiagaraNodeAssignment_0.SetVariables_#1
 ]
}
emitters {
 FEFakeTail {
  emitterUpdateScript [
   &EmitterState(Inactive Response=::NewEnumerator1, Life Cycle Mode=::NewEnumerator1, Loop Behavior=::NewEnumerator1, UseLoopDelay=false)
   &SpawnBurst_Instantaneous(
    Spawn Count=&MakeCustomIntFromBool(Bool=$User.IsFrontEnd, True INT=1, False INT=0)
    Use Spawn Probability=false
    UseLoopCountLimit=false
   )
   /SetVariables(Emitter.FELight:LinearColor=&MakeCustomLinearColorFromBool(
     Bool=$User.IsFrontEnd
     True Color=&MakeLinearColorFromVectorAndFloat(
      "Vector (RGB)"=&NormalizeVector(Vector To Normalize=<0.2 1 0>, Safe Normalize=true, Safe Fallback Threshold=0, Safe Fallback Vector=<0 0 0.001>)
      "Float (Alpha)"=1
     )
     False Color=<rgba 0 0 0 0>
    )) {isSetParameters=true}
  ]
  particleSpawnScript [
   &InitializeParticle(Color Mode=::NewEnumerator0, Lifetime=5, Mesh Scale Mode=::NewEnumerator4, UsePositionOffset=false, Write Material Random=true)
   &InitialMeshOrientation(Mesh Orientation Mode=::NewEnumerator7, Rotation=<0 0 0.25>, Use Rotation Vector=true)
  ]
  particleUpdateScript [
   &ParticleState()
   &DynamicMaterialParameters(
    Index 1 Param 1=&Multiply_Float(A=-55, B=1)
    Index 1 Param 3=1
    Index 1 Param 4=&FloatFromCurve(
     FloatCurve=DataInterface() {propertyValues {Curve {keys [{tangentMode=RCTM_User, time=0, value=0}, <at 0.1 0.07>, {tangentMode=RCTM_User, time=1.12, value=1}]}}}
     CurveIndex=$Particles.Age
    )
    Module.Param0WriteEnabled1=true
    Module.Param1WriteEnabled1=true
    Module.Param2WriteEnabled1=true
    Module.Param3WriteEnabled1=true
    Write Parameter Index 0=false
    Write Parameter Index 1=true
   )
   &DynamicMaterialParameters(
    Index 1 Param 1=&Multiply_Float(A=-45, B=1)
    Index 1 Param 3=1
    Index 1 Param 4=1
    Module.Param0WriteEnabled1=true
    Module.Param1WriteEnabled1=true
    Module.Param2WriteEnabled1=true
    Module.Param3WriteEnabled1=true
    Write Parameter Index 0=false
    Write Parameter Index 1=true
   ) {enabled=false, module=DynamicMaterialParameters001}
   &FN_Timer(
    Increment:NiagaraBool=&SetBoolByFloatComparison(A=$Particles.Age, B=0.2)
    Increment Multiplier:NiagaraFloat=&FloatFromCurve(
     FloatCurve=DataInterface() {propertyValues {Curve {keys [
         {interpMode=RCIM_Cubic, leaveTangent=6.8595, tangentMode=RCTM_Break, time=0.25, value=4}
         {arriveTangent=60.7369, interpMode=RCIM_Cubic, tangentMode=RCTM_Break, time=1.1, value=23}
        ]}}}
     CurveIndex=$Particles.Age
     Scale Curve=0.06
    )
   )
   /SetVariables(Particles.ColorA:LinearColor=&MakeLinearColor(R:NiagaraFloat=1, G:NiagaraFloat=1, B:NiagaraFloat=$Particles.FN_Timer.Value, A:NiagaraFloat=1)) {isSetParameters=true}
   /SetVariables(Particles.PositionB:NiagaraPosition=&Add_Vector(
     A=Linked() {linkedVariable {name=Particles.Position, type=NiagaraPosition}}
     B=&Add_Vector(
      A=&Multiply_VectorByFloat(Vector=&TransformVector(Vector=<1 0 0>, Destination Space=::World), Float=0)
      B=&Add_Vector(
       A=&Multiply_VectorByFloat(
        Vector=&TransformVector(Vector=<0 1 0>, Destination Space=::World)
        Float=&FloatFromCurve(
         FloatCurve=DataInterface() {propertyValues {Curve {keys [
             {arriveTangent=-0.9385, interpMode=RCIM_Cubic, leaveTangent=-0.9385, tangentMode=RCTM_User, time=0, value=1}
             {arriveTangent=-1, interpMode=RCIM_Cubic, leaveTangent=-1, time=0.1891, value=0.4412}
             <at 1 0>
            ]}}}
         Scale Curve=0
        )
       )
       B=&Multiply_VectorByFloat(Vector=&TransformVector(Vector=<0 0 1>, Destination Space=::World), Float=0)
      )
     )
    )) {isSetParameters=true}
  ]
  rendererClasses [&NiagaraSpriteRendererProperties, &NiagaraMeshRendererProperties]
  renderers [
   {rendererClass=&NiagaraSpriteRendererProperties}
   {rendererClass=&NiagaraMeshRendererProperties, rendererIndex=1}
  ]
 }
 RibbonTrail {
  bEnabled=false
  emitterUpdateScript [
   &EmitterState(
    Enable Distance Culling=true
    Life Cycle Mode=::NewEnumerator1
    Loop Behavior=::NewEnumerator1
    Loop Delay=0.55
    Loop Duration=0.75
    Max Distance Response=::KillImmediately
    MaxDistance=2000
    Module.bUseMaxDistance=true
    Module.bUseMinDistance=false
    Scalability Mode=::NewEnumerator1
    UseLoopDelay=true
   )
   &SpawnRate(
    SpawnRate=&MakeCustomFloatFromBool(
     Bool=$User.IsFrontEnd
     True Float=0
     False Float=&MakeCustomFloatFromBool(
      Bool=&SetBoolByIntComparison(A=$User.Deployed, B=1, Comparison Type=::NewEnumerator4)
      True Float=0
      False Float=65
     )
    )
    Use Spawn Probability=false
   )
  ]
  particleSpawnScript [
   &InitializeRibbon(Ribbon Width=200, UsePositionOffset=false, Write Material Random=true)
   &SkeletalMeshLocation(
    Bone Sampling Mode=::NewEnumerator4
    Interpolate Mesh Sampling=false
    MaskSpawnGroup=false
    Sampled Position Offset=<0 -60 0>
    Skeletal Mesh=$User.SKMesh
   )
  ]
  particleUpdateScript [
   &ParticleState()
   &ScaleRibbonWidth(Ribbon Width Scale=&FloatFromCurve(FloatCurve=DataInterface() {propertyValues {Curve {keys [{interpMode=RCIM_Cubic, tangentMode=RCTM_Break, time=0, value=1}, <at 1 0>]}}}))
   &ScaleColor(
    Scale Alpha=&FloatFromCurve(FloatCurve=DataInterface() {propertyValues {Curve {keys [{interpMode=RCIM_Cubic, leaveTangent=-3.0546, tangentMode=RCTM_Break, time=0, value=1}, <at 1 0>]}}})
    Scale RGB=&MakeVectorFromLinearColorRGB(Color=&ColorFromCurve(ColorCurve=DataInterface() {propertyValues {
        AlphaCurve {keys [<at 0 1 RCIM_Cubic>,<at 1 0 RCIM_Cubic>]}
        BlueCurve {keys [<at 0 0.15 RCIM_Cubic>,<at 1 1 RCIM_Cubic>]}
        GreenCurve {keys [<at 0 0.7875 RCIM_Cubic>,<at 1 1 RCIM_Cubic>]}
        RedCurve {keys [<at 0 1 RCIM_Cubic>,<at 1 1 RCIM_Cubic>]}
       }}))
    ScaleA=true
    ScaleRGB=true
   )
   &ScaleColor(
    Scale Alpha=&FloatFromCurve(FloatCurve=DataInterface() {propertyValues {Curve {keys [<at 0.7 0>,<at 0.9 1>]}}}, CurveIndex=$System.Age)
    ScaleA=true
    ScaleRGB=true
   ) {module=ScaleColor001}
   &self:Tod(
    Saturation:NiagaraFloat=&MakeCustomFloatFromBool(
     Bool=$User.IsFrontEnd
     True Float=1
     False Float=&FNTimeOfDayScalar(Night Scale=1, Morning Scale=1, Evening Scale=1.5)
    )
    Value:NiagaraFloat=&MakeCustomFloatFromBool(
     Bool=$User.IsFrontEnd
     True Float=1
     False Float=&FNTimeOfDayScalar(Night Scale=0.3, Morning Scale=0.3, Day Scale=0.7, Evening Scale=0.3)
    )
   ) {module=Tod}
  ]
  rendererClasses [&NiagaraSpriteRendererProperties, &NiagaraRibbonRendererProperties]
  renderers [
   {rendererClass=&NiagaraSpriteRendererProperties}
   {rendererClass=&NiagaraRibbonRendererProperties, rendererIndex=1}
  ]
 }
 SmallCloud {
  emitterUpdateScript [
   &EmitterState(Inactive Response=::NewEnumerator1, Life Cycle Mode=::NewEnumerator1, Loop Behavior=::NewEnumerator1, Loop Duration=2, UseLoopDelay=true)
   &SpawnBurst_Instantaneous(
    Spawn Count=&MakeCustomIntFromBool(
     Bool=$User.IsFrontEnd
     True INT=0
     False INT=&MakeIntFromFloat(Float=&MakeCustomFloatFromBool(
       Bool=&SetBoolByIntComparison(A=$User.Deployed, B=1, Comparison Type=::NewEnumerator4)
       True Float=0
       False Float=4
      ))
    )
    Spawn Time=1.65
    Use Spawn Probability=false
    UseLoopCountLimit=false
   )
   /SetVariables(Emitter.FELight:LinearColor=&MakeCustomLinearColorFromBool(Bool=$User.IsFrontEnd, True Color=<rgba 0 0 1 1>, False Color=<rgba 0 0 0 0>)) {isSetParameters=true}
  ]
  particleSpawnScript [
   &InitializeParticle(
    Color Mode=::NewEnumerator0
    Lifetime Max=0.9045
    Lifetime Min=0.5227
    Lifetime Mode=::NewEnumerator1
    Mesh Scale Mode=::NewEnumerator3
    Mesh Uniform Scale=&Lerp_Float(
     A=0.23
     B=0.55
     Alpha=&FloatFromCurve(
      FloatCurve=DataInterface() {propertyValues {Curve {keys [
          {interpMode=RCIM_Cubic, tangentMode=RCTM_User, time=0, value=0}
          {arriveTangent=3.0994, interpMode=RCIM_Cubic, leaveTangent=3.0994, tangentMode=RCTM_User, time=1, value=1}
         ]}}}
      CurveIndex=&RandomRangeFloat(Minimum=0, Maximum=1)
     )
    )
    UsePositionOffset=false
    Write Material Random=true
   )
   &SphereLocation(
    MaskSpawnGroup=false
    Override Seed=false
    Random Seed=1888
    Sphere Radius=1.5
    Surface Only=true
    Surface Only Band Thickness=0.01
    UseNonUniformScale=false
    UseOffset=false
    UseSphereOrientation=false
   ) {enabled=false}
   &ShapeLocation(Handle Radius=0.1, Large Radius=1.5, MaskSpawnGroup=false, Override Seed=false, Shape Primitive=::NewEnumerator3, UseOwnerScale=true)
   &InitialMeshOrientation(Use Rotation Vector=false)
   &AddVelocity(
    Module.Use Falloff Distance=false
    Origin Offset=<0 0 -0.5>
    Override Seed=false
    Velocity Mode=::NewEnumerator1
    Velocity Speed=&RandomRangeFloat(Minimum=1094.6323, Maximum=1710.5846)
   )
  ]
  particleUpdateScript [
   &Drag(Drag=&RandomRangeFloat(Minimum=7.5, Maximum=12.5795), Use Linear Drag=true, Use Rotational Drag=false)
   &GravityForce(Coordinate Space=::Local, Gravity=<-450 0 1530.3812>)
   &SolveForcesAndVelocity(Clamp Velocity=false, Limit Acceleration=false)
   &ParticleState()
   &UpdateMeshOrientation(
    Rotation Rate=&RandomRangeFloat(Minimum=0.8717, Maximum=3.1004)
    Rotation Vector=&RandomRangeVector(Minimum=<0 0 0>, Maximum=<1 1 1>)
   )
   &ScaleMeshSize(Scale Factor=&VectorFromFloat(Value=&FloatFromCurve(
      FloatCurve=DataInterface() {propertyValues {Curve {keys [
          {arriveTangent=29.3701, interpMode=RCIM_Cubic, leaveTangent=29.3701, tangentMode=RCTM_User, time=0, value=0}
          {arriveTangent=-7.214, interpMode=RCIM_Cubic, leaveTangent=-7.214, tangentMode=RCTM_User, time=0.2, value=1.875}
          {tangentMode=RCTM_User, time=0.9974, value=0}
         ]}}}
      Scale Curve=0.7
     )))
   &ScaleColor(
    Scale Alpha=&FloatFromCurve(FloatCurve=DataInterface() {propertyValues {Curve {keys [
         {interpMode=RCIM_Cubic, leaveTangent=5.7537, tangentMode=RCTM_Break, time=0, value=0}
         <at 0.3 1>
         {arriveTangent=-0.0405, interpMode=RCIM_Cubic, leaveTangent=-0.0405, tangentMode=RCTM_User, time=0.8, value=1}
         {arriveTangent=-2.2563, interpMode=RCIM_Cubic, leaveTangent=-2.2563, tangentMode=RCTM_User, time=1, value=0}
        ]}}})
    ScaleA=true
    ScaleRGB=true
   )
   &SkeletalMeshLocation(
    Bone Sampling Mode=::NewEnumerator4
    MaskSpawnGroup=false
    Orientation Sampling=::NewEnumerator1
    Position Sampling=::NewEnumerator2
    Sampled Position Offset=<0 20 -50>
    Skeletal Mesh=$User.SKMesh
   )
   /SetVariables(Particles.Positio2n:Vector3f=&Add_Vector(A=Linked() {linkedVariable {name=Particles.Position, type=NiagaraPosition}}, B=$Particles.SkeletalMeshLocation.SampledPosition)) {isSetParameters=true}
   &DynamicMaterialParameters(
    Index 3 Param 2=&MakeCustomFloatFromBool(Bool=$User.IsFrontEnd, True Float=0, False Float=0)
    Index 3 Param 3=&MakeCustomFloatFromBool(Bool=$User.IsFrontEnd, True Float=1.6, False Float=1)
    Module.Param0WriteEnabled3=true
    Module.Param1WriteEnabled3=true
    Module.Param2WriteEnabled3=true
    Module.Param3WriteEnabled3=true
    Write Parameter Index 0=false
    Write Parameter Index 3=true
   )
  ]
  rendererClasses [&NiagaraMeshRendererProperties]
  renderers [
   {rendererClass=&NiagaraMeshRendererProperties, rendererIndex=0}
   {rendererClass=&NiagaraMeshRendererProperties, rendererIndex=1}
  ]
 }
}
summary {
 systemName=NS_StaminaMale_Deploy
 emitters [
  {
   bEnabled=false
   emitterName=RibbonTrail
   rendererClasses [&NiagaraSpriteRendererProperties, &NiagaraRibbonRendererProperties]
  }
  {emitterName=SmallCloud, rendererClasses [&NiagaraMeshRendererProperties]}
  {emitterName=FEFakeTail, rendererClasses [&NiagaraSpriteRendererProperties, &NiagaraMeshRendererProperties]}
 ]
 userVariables [
  User.Deployed {type=NiagaraInt32}
  User.ForwardDot {type=NiagaraFloat, defaultValue=0, description=""}
  User.GlobalDelay {type=NiagaraFloat, defaultValue=0, description=""}
  User.IsFrontEnd {type=NiagaraBool}
  User.IsWhiteVersion {type=NiagaraBool}
  User.PMESH {
   type=NiagaraDataInterfaceSkeletalMesh
   defaultValue {
    struct=NiagaraExt_VariableValue_DataInterface
    value {dataInterface=&self:NiagaraDataInterfaceSkeletalMesh_2}
   }
  }
  User.SKMesh {
   type=NiagaraDataInterfaceSkeletalMesh
   defaultValue {
    struct=NiagaraExt_VariableValue_DataInterface
    value {dataInterface=&self:NiagaraDataInterfaceSkeletalMesh_0}
   }
  }
 ]
}
