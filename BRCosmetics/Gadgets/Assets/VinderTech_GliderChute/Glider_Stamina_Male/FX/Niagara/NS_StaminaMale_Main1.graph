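# Niagara effect for the Stamina Male glider: a deforming mesh, a cloud mesh and speed-line sprites.
# FN_FollowingDeform drives trailing stretch passed to materials; the cloud scales in with a settling wobble and feeds two feet positions to its material.
# Age-indexed curves animate deployment, with User.Deployed adding a large offset that jumps them to their end state.
# IsFrontEnd flags switch timing, offsets and spawn rate; gameplay speed lines scale with a system timer.
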NiagaraSystem /BRCosmetics/Gadgets/Assets/VinderTech_GliderChute/Glider_Stamina_Male/FX/Niagara/NS_StaminaMale_Main1
dependencies {
 usedDataInterfaces [&NiagaraDataInterfaceCurve]
 usedDynamicInputs [&SetBoolByFloatComparison]
 usedModules [&SystemState, &InterpolateOverTime]
}
emitters {
 FN_FollowingDeformTemplate001 {
  emitterUpdateScript [
   &EmitterState(
    Inactive Response=::NewEnumerator1
    Life Cycle Mode=::NewEnumerator1
    Loop Behavior=::NewEnumerator1
    Loop Duration Mode=::NewEnumerator1
    UseLoopDelay=false
   )
   &SpawnBurst_Instantaneous(Use Spawn Probability=false, UseLoopCountLimit=false)
   /SetVariables(Emitter.FELight:LinearColor=&MakeCustomLinearColorFromBool(Bool=$User.IsFrontEnd, True Color=<rgba 0 0 1 1>, False Color=<rgba 0 0 0 0>)) {isSetParameters=true}
  ]
  particleSpawnScript [
   &InitializeParticle(
    Color Mode=::NewEnumerator0
    Lifetime=5
    Mesh Scale=<1.2 1.1 1>
    Mesh Scale Mode=::NewEnumerator4
    Position Mode=::NewEnumerator0
    Sprite Size Mode=::NewEnumerator3
    Uniform Sprite Size=100
    Write Material Random=true
   )
   /SetVariables() {isSetParameters=true}
  ]
  particleUpdateScript [
   &ParticleState(Kill Particles When Lifetime Has Elapsed=false, Let Infinitely Lived Particles Die When Emitter Deactivates=true)
   /SetVariables() {isSetParameters=true}
   /SetVariables(Particles.FollowingDeform.Moving Average:Vector3f=&MakeCustomVectorFromBool(
     Bool=&Boolean_OrOperation(Bool1=$User.IsFrontEnd, Bool2=&SetBoolByIntComparison(A=$User.Deployed, B=1, Comparison Type=::NewEnumerator4))
     True Vector=$Particles.FollowingDeform.Moving Average
     False Vector=&Lerp_Vector(
      A=&Add_Vector(
       A=Linked() {linkedVariable {name=Engine.Owner.Position, type=NiagaraPosition}}
       B=&TransformVector(Vector=&Multiply_Vector(A=<-1000 800 1000>, B=$Engine.Owner.Scale), Destination Space=::World)
      )
      B=$Particles.FollowingDeform.Moving Average
      Alpha=&FloatFromCurve(
       FloatCurve=DataInterface() {propertyValues {Curve {keys [<at 1 0>,<at 2 1>]}}}
       CurveIndex=&Add_Float(A=$Particles.Age, B=&Multiply_FloatByInt(Float=100, Integer=$User.Deployed))
      )
     )
    )) {isSetParameters=true}
   /Game/Effects/Niagara/Modules/FollowingDeform/FN_FollowingDeform(
    CustomMeshOrientation=false
    Distance:NiagaraFloat=&MakeCustomFloatFromBool(
     Bool=$User.bIsFrontEnd
     True Float=&FloatFromCurve(FloatCurve=DataInterface() {propertyValues {Curve {keys [<at 1 0>,<at 1.3 1>]}}}, CurveIndex=$Particles.Age, Scale Curve=75)
     False Float=75
    )
    MaxAngle:NiagaraFloat=&FloatFromCurve(
     FloatCurve=DataInterface() {propertyValues {Curve {keys [<at 1 180>,<at 1.35 90>]}}}
     CurveIndex=&Add_Float(A=$Particles.Age, B=&Multiply_FloatByInt(Float=100, Integer=$User.Deployed))
    )
    MaxStretchDistance:NiagaraFloat=&MakeCustomFloatFromBool(
     Bool=&Boolean_OrOperation(Bool1=$User.IsFrontEnd, Bool2=&SetBoolByIntComparison(A=$User.Deployed, B=1, Comparison Type=::NewEnumerator4))
     True Float=&FloatFromCurve(FloatCurve=DataInterface() {propertyValues {Curve {keys [<at 1 0>,<at 1.3 1>]}}}, CurveIndex=$Particles.Age, Scale Curve=250)
     False Float=&FloatFromCurve(
      FloatCurve=DataInterface() {propertyValues {Curve {keys [
          {interpMode=RCIM_Cubic, leaveTangent=-1074.205, tangentMode=RCTM_Break, time=0, value=0}
          {arriveTangent=1821.9176, interpMode=RCIM_Cubic, leaveTangent=-5700.3721, tangentMode=RCTM_Break, time=0.75, value=2500}
          {interpMode=RCIM_Cubic, tangentMode=RCTM_Break, time=1.85, value=100}
         ]}}}
      CurveIndex=&Add_Float(A=$Particles.Age, B=&Multiply_FloatByInt(Float=100, Integer=$User.Deployed))
     )
    )
    MeshX+AxisAlignLS:Vector3f=<1 0 0>
    MeshZ+AxisAlignLS:Vector3f=<0 1 0>
    PlaneConstraint=true
    PlaneConstraintStrength:NiagaraFloat=&FloatFromCurve(
     FloatCurve=DataInterface() {propertyValues {Curve {keys [<at 1 0>,<at 1.35 1>]}}}
     CurveIndex=&Add_Float(A=$Particles.Age, B=&Multiply_FloatByInt(Float=100, Integer=$User.Deployed))
    )
    RootPosition:NiagaraPosition=&Add_Vector(
     A=Linked() {linkedVariable {name=Engine.Owner.Position, type=NiagaraPosition}}
     B=&TransformVector(
      Vector=&Multiply_Vector(
       A=&MakeVector(
        X=0
        Y=&FloatFromCurve(
         FloatCurve=DataInterface() {propertyValues {Curve {keys [<at 0 -0.5>,<at 1.5 1>]}}}
         CurveIndex=&Add_Float(A=$Particles.Age, B=&Multiply_FloatByInt(Float=100, Integer=$User.Deployed))
         Scale Curve=55
        )
        Z=-28
       )
       B=$Engine.Owner.Scale
      )
      Destination Space=::World
     )
    )
    Speed=2
    StretchRecoverSpeed:NiagaraFloat=&FloatFromCurve(
     FloatCurve=DataInterface() {propertyValues {Curve {keys [<at 0.8 0.2>,<at 1.35 2>]}}}
     CurveIndex=&Add_Float(A=$Particles.Age, B=&Multiply_FloatByInt(Float=100, Integer=$User.Deployed))
    )
   )
   &DynamicMaterialParameters(
    Index 0 Param 1=1.5
    Index 0 Param 2=0.2
    Index 0 Param 3=&FloatFromCurve(
     FloatCurve=DataInterface() {propertyValues {Curve {keys [<at 0 -1>,<at 1.45 0>]}}}
     CurveIndex=&Add_Float(A=$Particles.Age, B=&Multiply_FloatByInt(Float=100, Integer=$User.Deployed))
    )
    Index 0 Param 4=&FloatFromCurve(
     FloatCurve=DataInterface() {propertyValues {Curve {keys [<at 0 -1>,<at 1.45 0.5>]}}}
     CurveIndex=&Add_Float(A=$Particles.Age, B=&Multiply_FloatByInt(Float=100, Integer=$User.Deployed))
    )
    Index 1 Param 2=&MakeCustomFloatFromBool(Bool=$User.IsFrontEnd, True Float=0, False Float=0)
    Index 1 Param 3=&MakeCustomFloatFromBool(Bool=$User.IsFrontEnd, True Float=1.6, False Float=1)
    Index 2 Param 1=&MakeFloatFromVector(Vector=$Particles.FollowingDeform.Moving Average, Channel=::NewEnumerator0)
    Index 2 Param 2=&MakeFloatFromVector(Vector=$Particles.FollowingDeform.Moving Average, Channel=::NewEnumerator1)
    Index 2 Param 3=&MakeFloatFromVector(Vector=$Particles.FollowingDeform.Moving Average, Channel=::NewEnumerator2)
    Index 2 Param 4=$Particles.FollowingDeform.StretchDistance
    Index 3 Param 1=&MakeFloatFromVector(Vector=$Particles.FollowingDeform.ConstraintAxisWS, Channel=::NewEnumerator0)
    Index 3 Param 2=&MakeFloatFromVector(Vector=$Particles.FollowingDeform.ConstraintAxisWS, Channel=::NewEnumerator1)
    Index 3 Param 3=&MakeFloatFromVector(Vector=$Particles.FollowingDeform.ConstraintAxisWS, Channel=::NewEnumerator2)
    Index 3 Param 4=&Multiply_Float(
     A=&MakeFloatFromVector(Vector=$Engine.Owner.Scale, Channel=::NewEnumerator0)
     B=&FloatFromCurve(
      FloatCurve=DataInterface() {propertyValues {Curve {keys [<at 0 0.5>,<at 1.7 1>]}}}
      CurveIndex=&Add_Float(A=$Particles.Age, B=&Multiply_FloatByInt(Float=100, Integer=$User.Deployed))
     )
    )
    Module.Param0WriteEnabled=true
    Module.Param0WriteEnabled1=false
    Module.Param0WriteEnabled2=true
    Module.Param0WriteEnabled3=true
    Module.Param1WriteEnabled=true
    Module.Param1WriteEnabled1=true
    Module.Param1WriteEnabled2=true
    Module.Param1WriteEnabled3=true
    Module.Param2WriteEnabled=true
    Module.Param2WriteEnabled1=true
    Module.Param2WriteEnabled2=true
    Module.Param2WriteEnabled3=true
    Module.Param3WriteEnabled=true
    Module.Param3WriteEnabled1=false
    Module.Param3WriteEnabled2=true
    Module.Param3WriteEnabled3=true
    Write Parameter Index 1=true
    Write Parameter Index 2=true
    Write Parameter Index 3=true
   )
   &ScaleMeshSize(Scale Factor=$Engine.Owner.Scale)
   &ScaleMeshSize(Scale Factor=&VectorFromFloat(Value=&FloatFromCurve(
      FloatCurve=DataInterface() {propertyValues {Curve {keys [{interpMode=RCIM_Cubic, leaveTangent=1.3187, tangentMode=RCTM_Break, time=0.48, value=0}, {tangentMode=RCTM_User, time=1.5, value=1}]}}}
      CurveIndex=&Add_Float(A=$Particles.Age, B=&Multiply_FloatByInt(Float=100, Integer=$User.Deployed))
     ))) {module=ScaleMeshSize001}
   &ScaleMeshSize(Scale Factor=&MakeVector(
     X=&FloatFromCurve(
      FloatCurve=DataInterface() {propertyValues {Curve {keys [
          {interpMode=RCIM_Cubic, leaveTangent=0.7843, tangentMode=RCTM_Break, time=0.4, value=0.2}
          {arriveTangent=0.7835, interpMode=RCIM_Cubic, tangentMode=RCTM_Break, time=1.6, value=1}
         ]}}}
      CurveIndex=&Add_Float(A=$Particles.Age, B=&Multiply_FloatByInt(Float=100, Integer=$User.Deployed))
     )
     Y=&FloatFromCurve(
      FloatCurve=DataInterface() {propertyValues {Curve {keys [
          {interpMode=RCIM_Cubic, leaveTangent=0.7843, tangentMode=RCTM_Break, time=0.4, value=0.2}
          {arriveTangent=0.7835, interpMode=RCIM_Cubic, tangentMode=RCTM_Break, time=1.6, value=1}
         ]}}}
      CurveIndex=&Add_Float(A=$Particles.Age, B=&Multiply_FloatByInt(Float=100, Integer=$User.Deployed))
     )
     Z=1
    )) {module=ScaleMeshSize002}
   &ScaleColor(
    Scale Alpha=&FloatFromCurve(
     FloatCurve=DataInterface() {propertyValues {Curve {keys [<at 0 0>,<at 0.2 1>]}}}
     CurveIndex=&Add_Float(A=$Particles.Age, B=&Multiply_FloatByInt(Float=100, Integer=$User.Deployed))
     Scale Curve=&MakeCustomFloatFromBool(
      Bool=$User.IsFrontEnd
      True Float=&FloatFromCurve(
       FloatCurve=DataInterface() {propertyValues {Curve {keys [<at 1.15 0>,<at 1.3 1>]}}}
       CurveIndex=&Add_Float(A=$Particles.Age, B=&Multiply_FloatByInt(Float=100, Integer=$User.Deployed))
      )
      False Float=1
     )
    )
    ScaleA=true
    ScaleRGB=true
   )
   &self:TempFix(X+:Vector3f=<1 0 0>, Z+:Vector3f=<0 1 0>) {module=TempFix}
  ]
  rendererClasses [&NiagaraSpriteRendererProperties, &NiagaraMeshRendererProperties]
  renderers [
   {rendererClass=&NiagaraSpriteRendererProperties}
   {rendererClass=&NiagaraMeshRendererProperties, rendererIndex=1}
   {rendererClass=&NiagaraMeshRendererProperties, rendererIndex=2}
   {rendererClass=&NiagaraMeshRendererProperties, rendererIndex=3}
  ]
 }
 NE_StaminaGliderCloud {
  emitterUpdateScript [
   &EmitterState(
    Inactive Response=::NewEnumerator1
    Life Cycle Mode=::NewEnumerator1
    Loop Behavior=::NewEnumerator1
    Loop Delay=1.1
    Loop Duration Mode=::NewEnumerator1
    UseLoopDelay=false
   )
   &SpawnBurst_Instantaneous(Use Spawn Probability=false, UseLoopCountLimit=false)
   /SetVariables(Emitter.FELight:LinearColor=&MakeCustomLinearColorFromBool(Bool=$User.IsFrontEnd, True Color=<rgba 0 0 1 1>, False Color=<rgba 0 0 0 0>)) {isSetParameters=true}
  ]
  particleSpawnScript [
   &InitializeParticle(
    Color Mode=::NewEnumerator0
    Lifetime=5
    Mesh Scale=<0.18 0.18 0.185>
    Mesh Scale Mode=::NewEnumerator4
    Position Mode=::NewEnumerator1
    Position Offset=<0 -10 -21>
    Position Offset Coordinate Space=::Local
    UsePositionOffset=true
    Write Material Random=true
   )
   &InitialMeshOrientation(
    Mesh Axis To Orient=<0 1 0>
    Mesh Orientation Mode=::NewEnumerator2
    Orientation Coordinate Space=::Local
    Orientation Vector=&NormalizeVector(Vector To Normalize=<1 0 -0.025>, Safe Normalize=true, Safe Fallback Threshold=0, Safe Fallback Vector=<0 0 0.001>)
    Use Rotation Vector=false
   )
  ]
  particleUpdateScript [
   &ParticleState(Kill Particles When Lifetime Has Elapsed=false, Let Infinitely Lived Particles Die When Emitter Deactivates=true)
   &FN_Timer(Increment Multiplier:NiagaraFloat=&MakeCustomFloatFromBool(
     Bool=$User.IsFrontEnd
     True Float=&FloatFromCurve(
      FloatCurve=DataInterface() {propertyValues {Curve {keys [<at 0 2>,<at 0.4 2>,<at 1.6 0>]}}}
      CurveIndex=&Add_Float(A=$Particles.Age, B=&Multiply_FloatByInt(Float=100, Integer=$User.Deployed))
     )
     False Float=&Multiply_Float(A=&ClampFloat(Float=$System.InterpolateOverTime.InterpolateOverTime_Float.Moving Average), B=1)
    ))
   &ScaleMeshSize(Scale Factor=&VectorFromFloat(Value=&MakeCustomFloatFromBool(
      Bool=$User.IsFrontEnd
      True Float=&FloatFromCurve(
       FloatCurve=DataInterface() {propertyValues {Curve {keys [{interpMode=RCIM_Cubic, leaveTangent=1.8299, tangentMode=RCTM_Break, time=0, value=0}, {tangentMode=RCTM_User, time=1.5, value=1}]}}}
       CurveIndex=&Add_Float(A=$Particles.Age, B=&Multiply_FloatByInt(Float=100, Integer=$User.Deployed))
       Scale Curve=0.87
      )
      False Float=&FloatFromCurve(
       FloatCurve=DataInterface() {propertyValues {Curve {keys [{interpMode=RCIM_Cubic, leaveTangent=1.8299, tangentMode=RCTM_Break, time=0.4, value=0}, {tangentMode=RCTM_User, time=1.5, value=1}]}}}
       CurveIndex=&Add_Float(A=$Particles.Age, B=&Multiply_FloatByInt(Float=100, Integer=$User.Deployed))
       Scale Curve=0.87
      )
     )))
   &ScaleMeshSize(Scale Factor=&MakeVector(
     X=&Add_Float(
      A=1
      B=&Lerp_Float(
       A=&Sine(Normalized Angle=$Emitter.Age, Period=0.25, Scale=0.15, Bias=0)
       B=0
       Alpha=&FloatFromCurve(
        FloatCurve=DataInterface() {propertyValues {Curve {keys [<at 1.5606 1>, {interpMode=RCIM_Cubic, leaveTangent=0.7921, tangentMode=RCTM_Break, time=1.6606, value=0}, <at 3.9572 1>]}}}
        CurveIndex=&MakeCustomFloatFromBool(
         Bool=$User.IsFrontEnd
         True Float=&Add_Float(
          A=&Add_Float(A=$Particles.Age, B=&Multiply_FloatByInt(Float=100, Integer=$User.Deployed))
          B=0.35
         )
         False Float=&Add_Float(A=$Particles.Age, B=&Multiply_FloatByInt(Float=100, Integer=$User.Deployed))
        )
       )
      )
     )
     Y=&Add_Float(
      A=1
      B=&Lerp_Float(
       A=&Cosine(Normalized Angle=$Emitter.Age, Period=0.25, Scale=0.2, Bias=-0.1)
       B=0
       Alpha=&FloatFromCurve(
        FloatCurve=DataInterface() {propertyValues {Curve {keys [<at 1.5606 1>, {interpMode=RCIM_Cubic, leaveTangent=0.7921, tangentMode=RCTM_Break, time=1.6606, value=0}, <at 3.9572 1>]}}}
        CurveIndex=&MakeCustomFloatFromBool(
         Bool=$User.IsFrontEnd
         True Float=&Add_Float(
          A=&Add_Float(A=$Particles.Age, B=&Multiply_FloatByInt(Float=100, Integer=$User.Deployed))
          B=0.35
         )
         False Float=&Add_Float(A=$Particles.Age, B=&Multiply_FloatByInt(Float=100, Integer=$User.Deployed))
        )
       )
      )
     )
     Z=&Add_Float(
      A=1
      B=&Lerp_Float(
       A=&Sine(Normalized Angle=$Emitter.Age, Period=0.25, Scale=0.15, Bias=0)
       B=0
       Alpha=&FloatFromCurve(
        FloatCurve=DataInterface() {propertyValues {Curve {keys [<at 1.5606 1>, {interpMode=RCIM_Cubic, leaveTangent=0.7921, tangentMode=RCTM_Break, time=1.6606, value=0}, <at 3.9572 1>]}}}
        CurveIndex=&MakeCustomFloatFromBool(
         Bool=$User.IsFrontEnd
         True Float=&Add_Float(
          A=&Add_Float(A=$Particles.Age, B=&Multiply_FloatByInt(Float=100, Integer=$User.Deployed))
          B=0.35
         )
         False Float=&Add_Float(A=$Particles.Age, B=&Multiply_FloatByInt(Float=100, Integer=$User.Deployed))
        )
       )
      )
     )
    )) {module=ScaleMeshSize001}
   /SetVariables(Particles.FeetPos:Vector3f=/Niagara/DynamicInputs/Transforms/TransformPosition(
     Position=&Add_Vector(
      A=Linked() {linkedVariable {name=Particles.Position, type=NiagaraPosition}}
      B=&MakeCustomVectorFromBool(Bool=$User.IsFrontEnd, True Vector=<0 -25 1500>, False Vector=<0 -25 15>)
     )
     Source Space=::Local
     Destination Space=::World
    )) {isSetParameters=true}
   /SetVariables(Particles.Feet2Pos:Vector3f=&TransformPosition(
     Position=&Add_Vector(
      A=Linked() {linkedVariable {name=Particles.Position, type=NiagaraPosition}}
      B=&MakeCustomVectorFromBool(Bool=$User.IsFrontEnd, True Vector=<0 37 1500>, False Vector=<0 37 15>)
     )
     Source Space=::Local
     Destination Space=::World
    )) {isSetParameters=true}
   &DynamicMaterialParameters(
    Index 0 Param 1=&Multiply_Float(
     A=&Multiply_Float(
      A=&MakeFloatFromVector(Vector=$Particles.Scale, Channel=::NewEnumerator0)
      B=&MakeFloatFromVector(Vector=$Engine.Owner.Scale, Channel=::NewEnumerator0)
     )
     B=2
    )
    Index 0 Param 2=$Particles.FN_Timer.Value
    Index 0 Param 3=&Power(a=&MakeFloatFromVector(Vector=$Engine.Owner.Scale, Channel=::NewEnumerator0), b=0.5)
    Index 3 Param 2=&MakeCustomFloatFromBool(Bool=$User.IsFrontEnd, True Float=0, False Float=0)
    Index 3 Param 3=&MakeCustomFloatFromBool(Bool=$User.IsFrontEnd, True Float=1.6, False Float=1)
    Module.Param0WriteEnabled=true
    Module.Param0WriteEnabled3=true
    Module.Param1WriteEnabled=true
    Module.Param1WriteEnabled3=true
    Module.Param2WriteEnabled=true
    Module.Param2WriteEnabled3=true
    Module.Param3WriteEnabled=false
    Module.Param3WriteEnabled3=true
    Write Parameter Index 3=true
   )
   &DynamicMaterialParameters(
    Index 1 Param 1=&MakeFloatFromVector(Vector=$Particles.FeetPos, Channel=::NewEnumerator0)
    Index 1 Param 2=&MakeFloatFromVector(Vector=$Particles.FeetPos, Channel=::NewEnumerator1)
    Index 1 Param 3=&MakeFloatFromVector(Vector=$Particles.FeetPos, Channel=::NewEnumerator2)
    Index 1 Param 4=&FloatFromCurve(
     FloatCurve=DataInterface() {propertyValues {Curve {keys [<at 0 0>,<at 2 0>,<at 3 1>]}}}
     CurveIndex=&ClampFloat(
      Float=&Add_Float(A=$Particles.Age, B=&Multiply_FloatByInt(Float=100, Integer=$User.Deployed))
      Max=5
     )
    )
    Index 2 Param 1=&MakeFloatFromVector(Vector=$Particles.Feet2Pos, Channel=::NewEnumerator0)
    Index 2 Param 2=&MakeFloatFromVector(Vector=$Particles.Feet2Pos, Channel=::NewEnumerator1)
    Index 2 Param 3=&MakeFloatFromVector(Vector=$Particles.Feet2Pos, Channel=::NewEnumerator2)
    Module.Param0WriteEnabled1=true
    Module.Param0WriteEnabled2=true
    Module.Param1WriteEnabled1=true
    Module.Param1WriteEnabled2=true
    Module.Param2WriteEnabled1=true
    Module.Param2WriteEnabled2=true
    Module.Param3WriteEnabled1=true
    Module.Param3WriteEnabled2=false
    Write Parameter Index 0=false
    Write Parameter Index 1=true
    Write Parameter Index 2=true
   ) {module=DynamicMaterialParameters001}
   /SetVariables(Particles.Position2:NiagaraPosition=&Add_Vector(
     A=Linked() {linkedVariable {name=Particles.Position, type=NiagaraPosition}}
     B=&MakeCustomVectorFromBool(
      Bool=$User.IsFrontEnd
      True Vector=&MakeVector(
       X=0
       Y=0
       Z=&FloatFromCurve(
        FloatCurve=DataInterface() {propertyValues {Curve {keys [<at 0 0>,<at 0.3 0>,<at 0.35 1>,<at 1.45 1.1>,<at 1.75 0>]}}}
        CurveIndex=&Add_Float(A=$Particles.Age, B=&Multiply_FloatByInt(Float=100, Integer=$User.Deployed))
        Scale Curve=20
       )
      )
      False Vector=<0 0 0>
     )
    )) {isSetParameters=true}
  ]
  rendererClasses [&NiagaraMeshRendererProperties]
  renderers [
   {rendererClass=&NiagaraMeshRendererProperties, rendererIndex=0}
   {rendererClass=&NiagaraMeshRendererProperties, rendererIndex=1}
  ]
 }
 SpeedLine {
  emitterUpdateScript [
   &EmitterState(
    Delay First Loop Only=true
    Inactive Response=::NewEnumerator1
    Life Cycle Mode=::NewEnumerator1
    Loop Delay=1.5
    Loop Duration=3
    UseLoopDelay=true
   )
   &SpawnRate(
    SpawnRate=&MakeCustomFloatFromBool(
     Bool=$User.IsFrontEnd
     True Float=100
     False Float=&Multiply_Float(
      A=&Multiply_Float(A=230, B=1)
      B=&Lerp_Float(A=-0.8, B=1, Alpha=$System.FN_Timer.Value)
     )
    )
    Use Spawn Probability=false
   )
  ]
  particleSpawnScript [
   &InitializeParticle(
    Lifetime Max=0.12
    Lifetime Min=0.1
    Lifetime Mode=::NewEnumerator1
    Position Offset=<50 0 0>
    Sprite Size Max=<2 7>
    Sprite Size Min=<2 2>
    Sprite Size Mode=::NewEnumerator2
    UsePositionOffset=true
    Write Material Random=true
   )
   &TorusLocation(
    Align Torus With Axis=true
    Handle Radius=0
    Large Radius:NiagaraFloat=&RandomRangeFloat(Minimum=&MakeCustomFloatFromBool(Bool=$User.IsFrontEnd, True Float=100, False Float=370), Maximum=400)
    MaskSpawnGroup=false
    Module.Override Randomness=false
    Offset=&FN_FrontEndVector3Modifier(Gameplay=<0 100 0>, Front End=<0 -100 0>)
    Torus Axis=<0 1 0>
   )
   &AddVelocity(Velocity=&RandomRangeVector(Minimum=<0 500 0>, Maximum=<0 500 0>))
   &AddVelocityFromPoint(Module.Use Falloff Distance=false, Origin Offset=<0 -200 0>, Velocity Strength=1500)
   &CameraOffset(Camera Offset Amount=-10)
  ]
  particleUpdateScript [
   &SolveForcesAndVelocity(Clamp Velocity=false, Limit Acceleration=false)
   &ParticleState()
   &ScaleSpriteSizeBySpeed(Max Scale Factor=<1 50>, Min Scale Factor=<0.01 0.05>, Velocity Threshold=1)
   &ScaleColor(
    Scale Alpha=&Multiply_Float(
     A=&FloatFromCurve(FloatCurve=DataInterface() {propertyValues {Curve {keys [
          {interpMode=RCIM_Cubic, tangentMode=RCTM_User, time=0, value=0}
          {interpMode=RCIM_Cubic, tangentMode=RCTM_User, time=0.5, value=1}
          {tangentMode=RCTM_User, time=1, value=0}
         ]}}})
     B=&Multiply_Float(A=&RandomRangeFloat(Minimum=0.1, Maximum=0.3), B=1)
    )
    ScaleA=true
    ScaleRGB=true
   )
  ]
  rendererClasses [&NiagaraSpriteRendererProperties]
  renderers [{rendererClass=&NiagaraSpriteRendererProperties}]
 }
}
summary {
 systemName=NS_StaminaMale_Main1
 emitters [
  {emitterName=NE_StaminaGliderCloud, rendererClasses [&NiagaraMeshRendererProperties]}
  {emitterName=SpeedLine, rendererClasses [&NiagaraSpriteRendererProperties]}
  {
   emitterName=FN_FollowingDeformTemplate001
   rendererClasses [&NiagaraSpriteRendererProperties, &NiagaraMeshRendererProperties]
  }
 ]
 userVariables [
  User.bIsFrontEnd {type=NiagaraBool}
  User.bIsFrontEndPreview {type=NiagaraBool}
  User.Deployed {type=NiagaraInt32}
  User.ForwardDot {type=NiagaraFloat, defaultValue=0, description=""}
  User.GlobalDelay {type=NiagaraFloat, defaultValue=0, description=""}
  User.IsFrontEnd {type=NiagaraBool}
  User.IsWhiteVersion {type=NiagaraBool}
  User.PMESH {
   type=NiagaraDataInterfaceSkeletalMesh
   defaultValue {
    struct=NiagaraExt_VariableValue_DataInterface
    value {dataInterface=&self:NiagaraDataInterfaceSkeletalMesh_0}
   }
  }
 ]
}
